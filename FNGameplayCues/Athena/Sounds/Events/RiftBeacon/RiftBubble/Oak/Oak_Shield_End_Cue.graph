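# Sound cue for the end of the Oak shield, filed under the Rift Beacon rift bubble event gameplay cue sounds.
# A single wave player feeds the Oak_Shield_End_01 wave straight to the root at a strongly reduced volume, in the gadgets and items sound class.
# Overridden attenuation turns off spatialization and sets zero falloff distance, so the sound is non-directional with a hard cutoff at its radius.
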SoundCue /FNGameplayCues/Athena/Sounds/Events/RiftBeacon/RiftBubble/Oak/Oak_Shield_End_Cue
root=WavePlayer_0
nodes [{id=WavePlayer_0, properties {soundWaveAssetPtr=&/Oak_Shield_End_01}}]
properties {
 bOverrideAttenuation=true
 maxDistance=2048
 soundClassObject=&Fort_Gadgets_Items
 volumeMultiplier=0.1
 attenuationOverrides {bSpatialize=false, falloffDistance=0, attenuationShapeExtents {x=2048}}
}
