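# This sound cue supplies the music for entering the Oak rift bubble from a Rift Beacon, as its name indicates.
# It randomly picks one of two Oak music waves with equal weight, with one passed through a mixer to lower its level, under the Fort_Gadgets sound class.
# It borrows the safe zone shield attenuation and overrides concurrency to allow one instance per owner, blocking new plays.
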SoundCue /FNGameplayCues/Athena/Sounds/Events/RiftBeacon/RiftBubble/Oak/RiftBubble_Oak_Passthrough_Enter_Music_Cue
root=SoundClass_0
nodes [
 {id=Mixer_1, children [WavePlayer_1], properties {inputVolume [0.75]}}
 {id=Random_0, children [Mixer_1,WavePlayer_2], properties {weights [1,1]}}
 {id=SoundClass_0, children [Random_0], properties {soundClassOverride=&Fort_Gadgets}}
 {id=WavePlayer_1, properties {soundWaveAssetPtr=&/Oak_Music_02}}
 {id=WavePlayer_2, properties {soundWaveAssetPtr=&/Oak_Music_01}}
]
properties {
 attenuationSettings=/Game/Sounds/Attenuation/Fort_SafeZone_Shield_Athena_Att
 bOverrideConcurrency=true
 maxDistance=4000
 volumeMultiplier=0.4
 concurrencyOverrides {bLimitToOwner=true, maxCount=1, resolutionRule=PreventNew}
}
